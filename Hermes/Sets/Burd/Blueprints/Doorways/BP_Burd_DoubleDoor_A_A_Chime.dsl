(event EventBeginPlay
  (Networking|SwitchHasAuthority
    (:Authority
      (Utilities|Array|Add (Variables|Default|GetArrowLocations) (Transformation|GetWorldLocation (Variables|Default|GetArrowOutside)))
      (Utilities|Array|Add (Variables|Default|GetArrowLocations) (Transformation|GetWorldLocation (Variables|Default|GetArrowInside))))))

(event OnComponentEndOverlap(Box)
  (params (OverlappedComponent :type Primitive%20Component%20Object%20Reference) (OtherActor :type Actor%20Object%20Reference) (OtherComp :type Primitive%20Component%20Object%20Reference) (OtherBodyIndex :type Integer))
  (bind _asfort_pawn (Utilities|Casting|CastToFortPawn OtherActor))
  (Variables|Default|SetPlayerPawnRef _asfort_pawn)
  (Networking|SwitchHasAuthority
    (:Authority
      (if (Helper|IsDoorOpen)
        (bind _isoutsideheadinginside (CallFunction|IsPlayerOutsideHeadingInside (Variables|Default|GetPlayerPawnRef)))
        (if _isoutsideheadinginside
          (Utilities|FlowControl|DoOnce false
            (:Completed
              (CallFunction|Chime)
              (Utilities|FlowControl|Delay 5.0)
              (CallFunction|ResetChime))))))))

(event Custom|ResetChime
  (Utilities|FlowControl|DoOnce false
    (:Completed
      (CallFunction|Chime)
      (Utilities|FlowControl|Delay 5.0)
      (CallFunction|ResetChime))))

(event Custom|Chime
  (Networking|SwitchHasAuthority
    (:Remote
      (Audio|Components|Audio|Play (Variables|Default|GetChimeAudio) 0.0))))

(fn IsPlayerOutsideHeadingInside
  (params (PlayerPawn :type Actor%20Object%20Reference))
  (for _index (range 2)
    (Utilities|Array|Add (LocalVariables|GetLocalDistances) (Math|Vector|VectorLength (- (Transformation|GetActorLocation PlayerPawn) (Utilities|Array|Get(acopy) (Variables|Default|GetArrowLocations) _index)))))
  (return (== (Math|Float|MinOfFloatArray (LocalVariables|GetLocalDistances)) 0)))

(fn ConstructionScript
  (|CallParentFunction))
